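# Physics asset for the female Uproar pickaxe, a two-handed melee weapon skeletal mesh.
# It holds a single body on the Melee_TwoHanded_Handle bone, shaped as one large capsule offset up along the handle and rotated to align with it.
# It has no constraints, so the weapon simulates and collides as one rigid piece.
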PhysicsAsset /Game/Weapons/FORT_Melee/Pickaxe_Uproar_Female/Meshes/Uproar_Female_PhysicsAsset
bodies {Melee_TwoHanded_Handle [{shapeType=Capsule, length=76.6353, radius=30.4408, center=<1.1235 -0.8489 30.6218>, rotation=<rot -0.1584 -90 0.0425>}]}
constraints []
